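# NS_Money is the Niagara effect for the Money vehicle boost, spraying bill and coin sprites backward from an offset behind the car.
# Two sprite emitters, one explicitly GPU-simulated, each run four spawn groups with individual cones, tagged per particle alongside four renderers.
# Particles follow source movement, with color curves and painted-variant overrides driving tint.
# Per-particle bill and coin sizes plus per-renderer SubUV animation vary the sprites.
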NiagaraSystem /VehicleCosmetics/Boosts/Boost_Money/FX/NS_Money
dependencies {
 usedDataInterfaces [
  &NiagaraDataInterfaceColorCurve
  &NiagaraDataInterfaceVector2DCurve
  &NiagaraDataInterfaceCurve
  &NiagaraDataInterfaceSpriteRendererInfo
  &NiagaraDataInterfaceSkeletalMesh
 ]
 usedDynamicInputs []
 usedModules [
  &self:SystemSpawnScript.SystemScriptSource.SystemScriptGraph.NiagaraNodeAssignment_0.SetVariables_#1
  &SystemState
 ]
}
emitters {
 NE_VC_SingleBoost_Sprite_Parent {
  simTarget=GPUComputeSim
  emitterUpdateScript [
   &EmitterState()
   &SpawnRate(SpawnRate=90, Use Spawn Probability=false)
   &SpawnRate(SpawnGroup=1, SpawnRate=120, Use Spawn Probability=false) {module=SpawnRate001}
   &SpawnRate(SpawnGroup=2, SpawnRate=30, Use Spawn Probability=false) {module=SpawnRate002}
   &SpawnRate(SpawnGroup=3, SpawnRate=35, Use Spawn Probability=false) {module=SpawnRate003}
  ]
  particleSpawnScript [
   /SetVariables(Particles.SmVisTag:NiagaraInt32=&SpawnGroup()) {isSetParameters=true, module=SetVariables_#2}
   &VC_PaintedVariantsOverride(DefaultColor=<rgba 0.4115 0.4115 0.4115 1>) {enabled=false}
   &InitializeParticle(
    Color=$Particles.Color
    Position Offset=&AddVectorToPosition(Position=$User.DTCarOffset, Vector=<-200 0 20>)
    Position Offset Coordinate Space=::Local
    Sprite Size Mode=::NewEnumerator3
    Uniform Sprite Size=150
    UsePositionOffset=true
    Write Material Random=true
   )
   &AddVelocity(
    Cone Angle=&MergeHelper_FloatSwitch(0=25, 1=15, 2=10, 3=25, Index=$Emitter.SpawnGroup, Num=::NewEnumerator2)
    Cone Axis=&RandomRangeVector(
     Minimum=&MergeHelper_VectorSwitch(0=<1 0 -0.1>, 1=<1 0 0>, 2=<1 0 -0.15>, 3=<1 0 -0.2>, Index=&SpawnGroup(), Num=::NewEnumerator2)
     Maximum=<1 0 0>
    )
    Module.UseConeFalloff=false
    Module.UseVelDistribution=false
    Override Seed=false
    Velocity Mode=::NewEnumerator2
    Velocity Speed=-1000
   )
   &ShapeLocation(MaskSpawnGroup=false, Override Seed=false, Sphere Radius=50, UseOwnerScale=true)
  ]
  particleUpdateScript [
   &InheritSourceMovement(ApplyForce=false, ApplyPosition=true, ApplyPositionFalloff=false)
   &VC_PaintedVariantsOverride(
    DefaultColor=&ColorFromCurve(ColorCurve=DataInterface() {propertyValues {
       AlphaCurve {keys [<at 0 1 RCIM_Cubic>,<at 0.7355 1>,<at 1 0>]}
       BlueCurve {keys [<at 0.0881 0 RCIM_Cubic>, <at 0.2526 0.1748>, <at 0.3066 0.147>, {interpMode=RCIM_Cubic, leaveTangent=3.2996, time=0.3912, value=0.85}]}
       GreenCurve {keys [<at 0.0881 0.484 RCIM_Cubic>, <at 0.2526 0.7084>, <at 0.3066 0.7084>, {interpMode=RCIM_Cubic, leaveTangent=1.1265, time=0.3912, value=0.85}]}
       RedCurve {keys [<at 0.0881 0.5052 RCIM_Cubic>, <at 0.2526 0.3942>, <at 0.3066 0.3231>, {interpMode=RCIM_Cubic, leaveTangent=2.6073, time=0.3912, value=0.85}]}
      }})
    R1 DefaultColor=<rgba 0.2987 0.4427 0.1098 1>
    R2 DefaultColor:LinearColor=<rgba 1 0.669 0.16 1>
    Use R1 Painted Overrides=true
    Use R2 Painted Overrides=true
    Use R3 Painted Overrides=true
   ) {module=VC_PaintedVariantsOverride001}
   /SetVariables(
    Particles.BillSize:Vector2f=&RandomRangeVector2D(Minimum=<25 25>, Maximum=<100 100>)
    Particles.BillsColor:LinearColor=$Particles.R1 Color
    Particles.CoinSize:Vector2f=<25 25>
    Particles.Str8BillSize:Vector2f=&Vector2DFromCurve(Vector2Curve=DataInterface() {propertyValues {XCurve {keys [<at 0 25>,<at 1 25>]}, YCurve {keys [<at 0 75>,<at 0.5 0>]}}}, ScaleCurve=<1 1>)
   ) {isSetParameters=true, module=SetVariables_#3}
   &ParticleState()
   &ScaleSpriteSize(Uniform Curve Sprite Scale=DataInterface() {propertyValues {Curve {keys [{tangentMode=RCTM_User, time=0, value=1.25}, {tangentMode=RCTM_User, time=1, value=2}]}}})
   &SubUVAnimation(
    Direct Frame Index:NiagaraFloat=&FloatFromCurve(FloatCurve=DataInterface() {propertyValues {Curve {keys [<at 0 0 RCIM_Cubic>,<at 1 63>]}}})
    Output Alternate Renderer Binding=true
    "Sprite Renderer "=DataInterface() {propertyValues {SpriteRenderer=&self:NE_VC_SingleBoost_Sprite_Parent.NiagaraSpriteRendererProperties_0}}
    SubUV Animation Mode:ENiagara_SubUVLookupModeV2=::NewEnumerator5
   )
   &SubUVAnimation(
    Direct Frame Index:NiagaraFloat=&FloatFromCurve(FloatCurve=DataInterface() {propertyValues {Curve {keys [<at 0 0>,<at 1 63>]}}})
    Output Alternate Renderer Binding=true
    "Sprite Renderer "=DataInterface() {propertyValues {SpriteRenderer=&self:NE_VC_SingleBoost_Sprite_Parent.NiagaraSpriteRendererProperties_1}}
    SubUV Animation Mode:ENiagara_SubUVLookupModeV2=::NewEnumerator5
   ) {module=SubUVAnimation001}
   &SubUVAnimation(
    Direct Frame Index:NiagaraFloat=&FloatFromCurve(FloatCurve=DataInterface() {propertyValues {Curve {keys [<at 0 0 RCIM_Cubic>,<at 1 65 RCIM_Cubic>]}}})
    Output Alternate Renderer Binding=true
    "Sprite Renderer "=DataInterface() {propertyValues {SpriteRenderer=&self:NE_VC_SingleBoost_Sprite_Parent.NiagaraSpriteRendererProperties_2}}
    SubUV Animation Mode:ENiagara_SubUVLookupModeV2=::NewEnumerator5
   ) {module=SubUVAnimation002}
   &SolveForcesAndVelocity(Clamp Velocity=false, Limit Acceleration=false)
  ]
  rendererClasses [&NiagaraSpriteRendererProperties]
  renderers [
   {rendererClass=&NiagaraSpriteRendererProperties}
   {rendererClass=&NiagaraSpriteRendererProperties, rendererIndex=1}
   {rendererClass=&NiagaraSpriteRendererProperties, rendererIndex=2}
   {rendererClass=&NiagaraSpriteRendererProperties, rendererIndex=3}
  ]
 }
 NE_VC_SingleBoost_Sprite_Parent001 {
  emitterUpdateScript [
   &EmitterState()
   &SpawnRate(SpawnRate=30, Use Spawn Probability=false)
   &SpawnRate(SpawnGroup=1, SpawnRate=30, Use Spawn Probability=false) {module=SpawnRate001}
   &SpawnRate(SpawnGroup=2, SpawnRate=30, Use Spawn Probability=false) {module=SpawnRate002}
   &SpawnRate(SpawnGroup=3, SpawnRate=30, Use Spawn Probability=false) {module=SpawnRate003}
  ]
  particleSpawnScript [
   /SetVariables(Particles.LSmVisTag:NiagaraInt32=&SpawnGroup()) {isSetParameters=true, module=SetVariables_#2}
   &VC_PaintedVariantsOverride(DefaultColor=<rgba 0.4115 0.4115 0.4115 1>) {enabled=false}
   &InitializeParticle(
    Color=$Particles.Color
    Position Offset=&AddVectorToPosition(Position=$User.DTCarOffset, Vector=<-200 0 20>)
    Position Offset Coordinate Space=::Local
    Sprite Rotation Mode=::NewEnumerator2
    Sprite Size Mode=::NewEnumerator3
    Uniform Sprite Size=50
    UsePositionOffset=true
    Write Material Random=true
   )
   &AddVelocity(
    Cone Angle=&MergeHelper_FloatSwitch(0=20, 1=15, 2=10, 3=25, Index=$Emitter.SpawnGroup, Num=::NewEnumerator2)
    Cone Axis=&RandomRangeVector(
     Minimum=&MergeHelper_VectorSwitch(0=<1 0 -0.3>, 1=<1 0 0>, 2=<1 0 -0.15>, 3=<1 0 -0.2>, Index=&SpawnGroup(), Num=::NewEnumerator2)
     Maximum=<1 0 0>
    )
    Module.UseConeFalloff=false
    Module.UseVelDistribution=false
    Override Seed=false
    Velocity Mode=::NewEnumerator2
    Velocity Speed=-1200
   )
   &ShapeLocation(MaskSpawnGroup=false, Override Seed=false, Sphere Radius=60, UseOwnerScale=true)
  ]
  particleUpdateScript [
   &InheritSourceMovement(ApplyForce=false, ApplyPosition=true, ApplyPositionFalloff=false)
   &VC_PaintedVariantsOverride(
    DefaultColor=&ColorFromCurve(ColorCurve=DataInterface() {propertyValues {
       AlphaCurve {keys [<at 0 1 RCIM_Cubic>,<at 0.4585 1>,<at 1 0 RCIM_Cubic>]}
       BlueCurve {keys [<at 0 0.1337 RCIM_Cubic>,<at 0.2531 0.1876>,<at 0.2975 0.4062 RCIM_Cubic>]}
       GreenCurve {keys [<at 0 0.5587 RCIM_Cubic>,<at 0.2531 0.7083>,<at 0.2975 0.4062 RCIM_Cubic>]}
       RedCurve {keys [<at 0 0.6584 RCIM_Cubic>,<at 0.2531 0.1144>,<at 0.2975 0.4062 RCIM_Cubic>]}
      }})
    R1 DefaultColor=<rgba 0.2987 0.4427 0.1098 1>
    R2 DefaultColor:LinearColor=<rgba 1 0.669 0.16 1>
    Use R1 Painted Overrides=true
    Use R2 Painted Overrides=true
    Use R3 Painted Overrides=true
   ) {module=VC_PaintedVariantsOverride001}
   /SetVariables(
    Particles.BillSize:Vector2f=&RandomRangeVector2D(Minimum=<25 25>, Maximum=<100 100>)
    Particles.BillsColor:LinearColor=$Particles.R1 Color
    Particles.CoinSize:Vector2f=<25 25>
    Particles.Str8BillSize:Vector2f=&Vector2DFromCurve(Vector2Curve=DataInterface() {propertyValues {XCurve {keys [<at 0 25>,<at 1 25>]}, YCurve {keys [<at 0 75>,<at 0.5 0>]}}}, ScaleCurve=<1 1>)
   ) {isSetParameters=true, module=SetVariables_#3}
   &ParticleState()
   &ScaleSpriteSize(Uniform Curve Sprite Scale=DataInterface() {propertyValues {Curve {keys [{tangentMode=RCTM_User, time=0, value=2}, {tangentMode=RCTM_User, time=1, value=3}]}}})
   &SubUVAnimation(
    Direct Frame Index:NiagaraFloat=&FloatFromCurve(FloatCurve=DataInterface() {propertyValues {Curve {keys [<at 0 0>,<at 1 65>]}}})
    Output Alternate Renderer Binding=true
    "Sprite Renderer "=DataInterface() {propertyValues {SpriteRenderer=&self:NE_VC_SingleBoost_Sprite_Parent001_0.NiagaraSpriteRendererProperties_0}}
    SubUV Animation Mode:ENiagara_SubUVLookupModeV2=::NewEnumerator5
   )
   &SubUVAnimation(
    Output Alternate Renderer Binding=true
    "Sprite Renderer "=DataInterface() {propertyValues {SpriteRenderer=&self:NE_VC_SingleBoost_Sprite_Parent001_0.NiagaraSpriteRendererProperties_1}}
    UseEndFrame=false
    UseStartFrame=false
   ) {module=SubUVAnimation001}
   &SubUVAnimation(
    Direct Frame Index:NiagaraFloat=&FloatFromCurve(FloatCurve=DataInterface() {propertyValues {Curve {keys [<at 0 0 RCIM_Cubic>,<at 1 65 RCIM_Cubic>]}}})
    Output Alternate Renderer Binding=true
    "Sprite Renderer "=DataInterface() {propertyValues {SpriteRenderer=&self:NE_VC_SingleBoost_Sprite_Parent001_0.NiagaraSpriteRendererProperties_2}}
    SubUV Animation Mode:ENiagara_SubUVLookupModeV2=::NewEnumerator5
   ) {module=SubUVAnimation002}
   &SolveForcesAndVelocity(Clamp Velocity=false, Limit Acceleration=false)
  ]
  rendererClasses [&NiagaraSpriteRendererProperties]
  renderers [
   {rendererClass=&NiagaraSpriteRendererProperties}
   {rendererClass=&NiagaraSpriteRendererProperties, rendererIndex=1}
   {rendererClass=&NiagaraSpriteRendererProperties, rendererIndex=2}
   {rendererClass=&NiagaraSpriteRendererProperties, rendererIndex=3}
  ]
 }
}
summary {
 systemName=NS_Money
 emitters [
  {emitterName=NE_VC_SingleBoost_Sprite_Parent, simTarget=GPUComputeSim, rendererClasses [&NiagaraSpriteRendererProperties]}
  {emitterName=NE_VC_SingleBoost_Sprite_Parent001, rendererClasses [&NiagaraSpriteRendererProperties]}
 ]
 userVariables [
  User.bIsFrontendPreview {type=NiagaraBool}
  User.bIsPainted {type=NiagaraBool}
  User.ColorVariantID {type=NiagaraInt32}
  User.DTCarOffset {type=NiagaraPosition, description="", defaultValue {}}
  User.GameModeID {type=NiagaraInt32}
  User.PaintedColor {type=LinearColor, description="", defaultValue {struct=LinearColor, value=<rgba 1 1 1 1>}}
  User.PaintedScalar {type=NiagaraFloat, defaultValue=1, description=""}
  User.PortCount {type=NiagaraInt32, defaultValue=1}
  User.SingleBoostPos {type=NiagaraPosition, description="", defaultValue {}}
 ]
}
